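# Animation Blueprint for the masked MusketSlinger head on the Fortnite 3L head skeleton, with logic for muting mouth facial curves.
# A Sequence reads the pawn's Stage2 cosmetic variant into a Branch with no following actions.
# Its second output passes an unwired-condition Branch to set the mouth mute curve expression asset, independently of the variant.
# No event is shown wired into the Sequence, and no AnimGraph is included.
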AnimBlueprint /BRCosmetics/Characters/Player/Male/Medium/Heads/M_MED_MusketSlinger_Head/Meshes/M_MED_MusketSlinger_MaskOn_Head_AnimBP
target_skeleton=/Game/Characters/Player/Common/Fortnite_Base_Head/Fortnite_Base_3L_Head_Skeleton
graphs [EventGraph {nodes [
   n1 = Sequence()
   > n2 = Cast To FortPlayerPawn(Object=GetOwningActor())
   > n3 = GetVariantChannelActiveIndexForItemActor!(
    execute
    Item:Athena Cosmetic Item Definition Object Reference=&Character_MusketSlinger
    Actor:Actor Object Reference=@n2
    Channel:Gameplay Tag Structure=(TagName="Cosmetics.Variant.Property.Stage2")
    VariantList:Array of Gameplay Tag Structures=Make Array("[0]"=(TagName="Cosmetics.Variant.Property.Stage2"))
    bErrorIfChannelMissing:bool=false
    bErrorIfTagNotInArray:bool=false
   ) -> (then, Index:int=0, bIndexIsValid:bool, bFoundChannel:bool)
   > Branch(Condition=@n3.bIndexIsValid)
   Branch(execute=@n1.then_1)
   > Set facial_curve_mute_expression(facial_curve_mute_expression:Curve Expressions Data Asset Object Reference=/BRCosmetics/Characters/Player/Common/Fortnite_Base_Head/Facials/CurveMappings/FN_3L_Mouth_Curve_Mute_Expr) -> (Output_Get:Curve Expressions Data Asset Object Reference)
   TryGetPawnOwner()
   Event BlueprintUpdateAnimation()
   Event OnCharacterCustomizationLoaded()
  ]}]
variables []
